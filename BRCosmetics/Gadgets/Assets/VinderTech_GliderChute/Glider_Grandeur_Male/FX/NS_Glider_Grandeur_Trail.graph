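# Niagara trail effect for the Grandeur Male glider, built from one emitter that bursts eight persistent mesh particles, each in its own spawn group.
# Particles follow the user skeletal mesh, with per-group offsets, rotations, material parameters and alpha chosen from arrays by group index.
# A front-end flag swaps in flipped placement values and a fixed fade input, replacing the system integrator value that otherwise drives opacity.
# A smoothed moving average of particle position is stored as a particle attribute.
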NiagaraSystem /BRCosmetics/Gadgets/Assets/VinderTech_GliderChute/Glider_Grandeur_Male/FX/NS_Glider_Grandeur_Trail
dependencies {
 usedDataInterfaces [&NiagaraDataInterfaceArrayFloat3, &NiagaraDataInterfaceArrayFloat, &NiagaraDataInterfaceCurve]
 usedDynamicInputs []
 usedModules [&SystemState, &FN_Integrator]
}
emitters {Trail {
  emitterSpawnScript [/SetVariables(Emitter.Trail Noise=0.1, Emitter.Trail U Speed=-1, Emitter.Trail U Tile=0.5) {isSetParameters=true}]
  emitterUpdateScript [
   &EmitterState(Life Cycle Mode=::NewEnumerator1, Loop Behavior=::NewEnumerator1, Loop Duration Mode=::NewEnumerator1, UseLoopDelay=false)
   &SpawnBurst_Instantaneous(Spawn Count=8, Use Spawn Probability=false)
  ]
  particleSpawnScript [
   &FN_SetSpawnGroup(Number of Spawn Groups=8)
   &InitializeParticle(Color=<rgba 1 1 1 0.3>, Mesh Scale Mode=::NewEnumerator4, UsePositionOffset=false, Write Material Random=true)
   &InitialMeshOrientation(
    Rotation=&SelectVectorFromArray(
     Direct Array Index=$Particles.SpawnGroup
     Vector Selection Array=DataInterface() {propertyValues {FloatData [<0 0.05 0.25>, <0 -0.05 0.25>, <0 0.05 0.25>, <0 -0.05 0.25>, <0 0 0.25>, <0 0 0.25>, <0 0 0.25>, <0 0 0.25>]}}
    )
    Use Orientation Vector=true
    Use Rotation Vector=true
   )
   &InitialMeshOrientation(
    Mesh Orientation Mode=::NewEnumerator2
    Rotation=&MakeCustomVectorFromBool(Bool=$User.bisFrontEnd, True Vector=<0.5 0 0>, False Vector=<0 0 0>)
    Use Rotation Vector=true
   ) {module=InitialMeshOrientation002}
   &InitialMeshOrientation(
    Rotation=&MakeCustomVectorFromBool(
     Bool=$User.bisFrontEnd
     True Vector=&SelectVectorFromArray(
      Direct Array Index=$Particles.SpawnGroup
      Vector Selection Array=DataInterface() {propertyValues {FloatData [<0 0 0>, <0 0.5 0>, <0 0 -0.03>, <-0.03 0.5 0>, <0 0.25 -0.03>, <0 0 -0.03>, <-0.03 0.5 0>, <-0.03 0.25 0>]}}
     )
     False Vector=&SelectVectorFromArray(
      Direct Array Index=$Particles.SpawnGroup
      Vector Selection Array=DataInterface() {propertyValues {FloatData [<0 0 0>, <0 0.5 0>, <0 0 0.03>, <0.03 0.5 0>, <0 0.25 0.03>, <0 0 0.03>, <0.03 0.5 0>, <0.03 0.25 0>]}}
     )
    )
    Use Orientation Vector=true
    Use Rotation Vector=true
   ) {module=InitialMeshOrientation001}
  ]
  particleUpdateScript [
   &ParticleState(Kill Particles When Lifetime Has Elapsed=false, Let Infinitely Lived Particles Die When Emitter Deactivates=true)
   &SkeletalMeshLocation(
    Apply Sampled Orientation to Mesh Particles=true
    Bone Sampling Mode=::NewEnumerator4
    Inherit Velocity=false
    Interpolate Mesh Sampling=false
    MaskSpawnGroup=false
    Sampled Position Offset=&MakeCustomVectorFromBool(
     Bool=$User.bisFrontEnd
     True Vector=&SelectVectorFromArray(
      Direct Array Index=$Particles.SpawnGroup
      Vector Selection Array=DataInterface() {propertyValues {FloatData [<-10 50 110>, <10 50 110>, <-15 40 105>, <15 40 105>, <-5 80 75>, <10 -100 100>, <-10 -100 100>, <5 80 75>]}}
     )
     False Vector=&SelectVectorFromArray(
      Direct Array Index=$Particles.SpawnGroup
      Vector Selection Array=DataInterface() {propertyValues {FloatData [<-10 -50 110>, <10 -50 110>, <-15 -40 105>, <15 -40 105>, <-5 -80 75>, <10 100 100>, <-10 100 100>, <5 -80 75>]}}
     )
    )
    Sampled Position Offset Coordinate Space:ENiagaraMeshTransforms=::NewEnumerator3
    Skeletal Mesh=$User.Skeletal Mesh
   ) {module=SkeletalMeshLocation001}
   &InterpolateOverTime(Rate Of Change=10, Target Vector Value:NiagaraPosition=$Particles.Position, Variable Type=::NewEnumerator2)
   /SetVariables(Particles.InterpolateDynamic:Vector4f=&MakeFloat4FromVectorAndFloat("Vector (RGB)"=$Output.InterpolateOverTime.Moving Average, "Float (Alpha)"=1)) {isSetParameters=true}
   &DynamicMaterialParameters(
    Index 0 Param 1=&SelectFloatFromArray(
     Direct Array Index=$Particles.SpawnGroup
     Float Selection Array=DataInterface() {propertyValues {FloatData [0.26,0.4832,0.54,0.25,1.2447,0.4,1.926,4.0704]}}
    )
    Index 0 Param 2=&SelectFloatFromArray(
     Direct Array Index=$Particles.SpawnGroup
     Float Selection Array=DataInterface() {propertyValues {FloatData [0.3,0.1832,0.08,0.65,0.26,0.17,0.48,0.24]}}
    )
    Index 0 Param 3=&SelectFloatFromArray(Direct Array Index=$Particles.SpawnGroup, Float Selection Array=DataInterface() {propertyValues {FloatData [0.7,0.7,0.6,0.6,1,0.6,0.6,1]}})
    Index 2 Param 1=&Multiply_Float(
     A=0.3
     B=&SelectFloatFromArray(
      Direct Array Index=$Particles.SpawnGroup
      Float Selection Array=DataInterface() {propertyValues {FloatData [0.8,0.83,0.64,0.6797,0.5,0.7366,0.71,0.5]}}
     )
    )
    Index 2 Param 2=3
    Index 2 Param 3=3
    Module.Param0WriteEnabled=true
    Module.Param0WriteEnabled2=true
    Module.Param1WriteEnabled=true
    Module.Param1WriteEnabled2=true
    Module.Param2WriteEnabled=true
    Module.Param2WriteEnabled2=true
    Module.Param3WriteEnabled=true
    Module.Param3WriteEnabled2=true
    Write Parameter Index 2=true
   )
   &ScaleColor(
    Scale Alpha=&MakeCustomFloatFromBool(
     Bool=$User.bisFrontEnd
     True Float=&SelectFloatFromArray(Direct Array Index=$Particles.SpawnGroup, Float Selection Array=DataInterface() {propertyValues {FloatData [1,1,0.6,0.6,0.5,0.5,0.5,0.5]}})
     False Float=&SelectFloatFromArray(Direct Array Index=$Particles.SpawnGroup, Float Selection Array=DataInterface() {propertyValues {FloatData [1,1,0.6,0.6,1,0.5,0.5,1]}})
    )
    ScaleA=true
    ScaleRGB=true
   )
   &ScaleColor(
    Scale Alpha=&FloatFromCurve(
     FloatCurve=DataInterface() {propertyValues {Curve {keys [<at 0 0 RCIM_Cubic>, {arriveTangent=2.0248, interpMode=RCIM_Cubic, leaveTangent=2.0248, tangentMode=RCTM_User, time=1, value=1}]}}}
     CurveIndex=&MakeCustomFloatFromBool(Bool=$User.bisFrontEnd, True Float=0.6, False Float=$System.FN_Integrator001.Value)
     Scale Curve=0.6
    )
    ScaleA=true
    ScaleRGB=true
   ) {module=ScaleColor001}
   &ScaleColor(Scale Alpha=0.5, ScaleA=true, ScaleRGB=true) {module=ScaleColor002}
  ]
  rendererClasses [&NiagaraMeshRendererProperties]
  renderers [{rendererClass=&NiagaraMeshRendererProperties, rendererIndex=0}]
 }}
summary {
 systemName=NS_Glider_Grandeur_Trail
 emitters [{emitterName=Trail, rendererClasses [&NiagaraMeshRendererProperties]}]
 userVariables [
  User.bisFrontEnd {type=NiagaraBool}
  User.bIsFrontEndPreview {type=NiagaraBool}
  User.bisFullyDeployed {type=NiagaraBool}
  User.ForwardDot {type=NiagaraFloat, defaultValue=0, description=""}
  User.Skeletal Mesh {
   type=NiagaraDataInterfaceSkeletalMesh
   defaultValue {
    struct=NiagaraExt_VariableValue_DataInterface
    value {dataInterface=&self:NiagaraDataInterfaceSkeletalMesh_0}
   }
  }
 ]
}
